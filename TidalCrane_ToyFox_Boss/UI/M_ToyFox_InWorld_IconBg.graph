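# UI material for the Toy Fox boss in-world icon, compositing a channel-packed icon texture over a hexagonal background with alpha-composite blending.
# Texture channels become antialiased icon and outline masks, while polygon SDFs supply the badge, its border and an expanding, fading pulse.
# Time drives a stepped rotation "thwack", an optional scale pulse and intermittent noise-based UV distortion, blended in by a Glitch parameter.
# Gradient masks, a time-animated cosine palette and named reroutes feed color and alpha composites.
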Material /TidalCrane_ToyFox_Boss/UI/M_ToyFox_InWorld_IconBg
graph {
 properties {bAutomaticallySetUsageInEditor=false, blendMode=BLEND_AlphaComposite, materialDomain=MD_UI, usedShadingModels=Default Lit}
 nodes [
  n1 = TextureCoordinate()
  n2 = NamedRerouteUsage()
  n3 = ScalarParameter() {parameterName=Glitch?}
  n4 = Time()
  n5 = Constant() {r=0.5}
  n6 = SmoothStep(Value=Frac(None=Multiply(A=@n4, B=@n5) {constB=0.5})) {constMin=0.9}
  n7 = MaterialFunctionCall(
   UVs=MaterialFunctionCall(UV=LinearInterpolate(A=@n1, B=@n2, Alpha=@n3)) {materialFunction=&MF_UI_UVScaling}.Scaled to Fit
   Texture Scale=LinearInterpolate(B=LinearInterpolate(Alpha=@n6) {constA=0.5, constB=0.6}, Alpha=ScalarParameter() {parameterName=Scale Pulse}) {constA=0.5}
  ) {materialFunction=&ScaleUVsByCenter}
  n8 = Step(Y=MaterialFunctionCall(Speed=Constant() {desc=Icon Thwack Speed}) {materialFunction=&DebugTimeSine}.Result) {constX=0.5}
  n9 = LinearInterpolate(Alpha=@n8) {constA=0.08, constB=0}
  n10 = TextureSampleParameter2D(UVs=Add(
    A=MaterialFunctionCall(UVs=@n7.UVs, "Rotation Angle (0-1)"=@n9) {materialFunction=&CustomRotator}.Rotated Values
    B=Constant2Vector() {g=0.052, r=0.028}
   )) {parameterName=Texture Input, texture=&/T_UI_ToyFox_Icon_Melee}
  n11 = MaterialFunctionCall(DistanceField=@n10.R, Threshold=Constant() {r=0.5}, Smoothing=Constant() {r=0.1}) {materialFunction=&MF_UI_AntiAliasing}
  n12 = MaterialFunctionCall(DistanceField=@n10.B, Threshold=Constant() {r=0.4}, Smoothing=Constant()) {materialFunction=&MF_UI_AntiAliasing}
  n13 = Multiply(A=@n12.Outer)
  Add(A=@n11.Outer, B=@n13)
  Constant3Vector() {constant=<rgba 0.0365 0.0365 0.0365 1>}
  Constant() {r=0.5}
  MaterialFunctionCall!(UVs=NamedRerouteUsage()) {materialFunction=/FortUILibrary/MaterialFunctions/Basics/MF_UI_9SliceUVs}
  n14 = Saturate(None=NamedRerouteUsage())
  n15 = ScalarParameter() {defaultValue=0.1, parameterName=Pulse Master Opacity}
  n16 = Add(A=Saturate(None=Multiply(A=@n14, B=@n15)), B=Multiply(A=NamedRerouteUsage()) {constB=0.4})
  Multiply(A=@n16)
  n17 = TextureCoordinate()
  NamedRerouteDeclaration(None=ComponentMask(None=@n17) {g=true}) {name=G Mask, nodeColor=<rgba 0 0 0 1>}
  n18 = Time()
  n19 = MaterialFunctionCall!(
   Gradient=NamedRerouteUsage()
   a
   b
   c
   d=MaterialFunctionCall(R=Multiply(A=@n18), G=Constant() {r=0.3}, B=Constant()) {materialFunction=&AppendMany}.RGB
  ) {materialFunction=/FortUILibrary/MaterialFunctions/Color/MF_UI_CosineGradient}
  NamedRerouteDeclaration(None=Add(A=@n19.Result) {constB=0.5}) {name=Party Time Colorz, nodeColor=<rgba 0 0 0 1>}
  n20 = Add(A=@n11.Outer, B=@n12.Outer)
  n21 = Constant() {r=0.01}
  n22 = Constant()
  NamedRerouteDeclaration(None=Multiply(
    A=OneMinus(None=@n20)
    B=MaterialFunctionCall(DistanceField=@n10.G, Threshold=@n21, Smoothing=@n22) {materialFunction=&MF_UI_AntiAliasing}.Outer
   )) {name=IconOutline, nodeColor=<rgba 0 0 0 1>}
  NamedRerouteDeclaration(None=Saturate(None=@n11.Outer)) {name=Icon, nodeColor=<rgba 0 0 0 1>}
  n23 = TextureCoordinate()
  n24 = NamedRerouteUsage()
  n25 = ScalarParameter() {parameterName=Glitch?}
  n26 = Fmod(A=Multiply(A=Time()), B=Constant() {r=1})
  n27 = MaterialFunctionCall!(
   UVs=Add(A=LinearInterpolate(A=@n23, B=@n24, Alpha=@n25), B=Constant2Vector())
   Radius=LinearInterpolate(Alpha=@n26) {constA=0.25}
   Sides=Constant() {r=6}
   InnerAngle
  ) {materialFunction=/FortUILibrary/MaterialFunctions/Shapes/SDFs/MF_SDF_RegularPolygon}
  n28 = MaterialFunctionCall(DistanceField=@n27.SDF, Smoothing=LinearInterpolate(Alpha=@n26) {constA=0.8, constB=0.3}) {materialFunction=&MF_UI_AntiAliasing}
  n29 = Saturate(None=Multiply(
    A=Multiply(A=@n28.Outer)
    B=MaterialFunctionCall(DistanceField=@n27.SDF) {materialFunction=&MF_UI_AntiAliasing}.Inner
   ))
  n30 = LinearInterpolate(A=Constant() {r=0.25}, B=Constant(), Alpha=@n26)
  NamedRerouteDeclaration(None=Multiply(A=Multiply(A=@n29, B=@n30), B=ScalarParameter() {defaultValue=3, parameterName=Pulse Opacity})) {name=Pulse, nodeColor=<rgba 0 0 0 1>}
  n31 = Constant() {r=0.45}
  n32 = Constant() {r=0.604}
  NamedRerouteDeclaration(None=ComponentMask(None=MaterialFunctionCall(UVs=@n17, Rotation Center=@n31, "Rotation Angle (0-1)"=@n32) {materialFunction=&CustomRotator}.Rotated Values) {g=true}) {name=Angled Gradient Mask, nodeColor=<rgba 0 0 0 1>}
  n33 = NamedRerouteUsage()
  n34 = ScalarParameter() {defaultValue=0.1, parameterName=Pulse Master Opacity}
  n35 = Constant3Vector() {constant=<rgba 0 0 0 1>}
  n36 = Multiply(A=NamedRerouteUsage())
  n37 = Multiply(A=Add(A=NamedRerouteUsage()) {constB=0.2}, B=Multiply(A=NamedRerouteUsage()))
  n38 = LinearInterpolate(
   A=Multiply(A=NamedRerouteUsage(), B=Multiply(A=@n33, B=@n34))
   B=Add(A=Multiply(A=@n35, B=@n36), B=Saturate(None=@n37))
   Alpha=NamedRerouteUsage()
  )
  n39 = Multiply(A=Constant3Vector() {constant=<rgba 0.0365 0.0365 0.0365 1>}, B=NamedRerouteUsage())
  n40 = NamedRerouteUsage()
  n41 = NamedRerouteUsage()
  n42 = Add(A=Add(A=Saturate(None=@n38), B=@n39), B=Multiply(A=@n40, B=@n41))
  n43 = Multiply(A=NamedRerouteUsage(), B=Constant3Vector() {constant=<rgba 0.2214 0.2214 0.2214 1>})
  n44 = NamedRerouteUsage()
  n45 = Multiply(A=@n43, B=@n44) {constB=0.1}
  n46 = LinearInterpolate(A=@n39, B=@n40, Alpha=@n41) {constAlpha=1}
  NamedRerouteDeclaration(None=LinearInterpolate(A=Add(A=@n42, B=@n45), B=LinearInterpolate(A=@n46, B=@n43, Alpha=@n44), Alpha=NamedRerouteUsage()) {constAlpha=1}) {name=Color Stuff, nodeColor=<rgba 0 0 0 1>}
  NamedRerouteDeclaration(None=@n20) {name=Icon Composite, nodeColor=<rgba 0 0 0 1>}
  n47 = Add(A=Multiply(A=NamedRerouteUsage()), B=Multiply(A=NamedRerouteUsage()))
  n48 = Multiply(A=NamedRerouteUsage()) {constB=0.9}
  NamedRerouteDeclaration(None=Add(A=@n16, B=Add(A=@n47, B=@n48))) {name=Alpha Composite, nodeColor=<rgba 0 0 0 1>}
  NamedRerouteDeclaration(None=@n8) {name=Timed Business, nodeColor=<rgba 0 0 0 1>}
  n49 = TextureCoordinate()
  n50 = Time()
  n51 = ScalarParameter() {defaultValue=1, parameterName=Glitch In}
  n52 = TextureSample(UVs=Panner(Coordinate=TextureCoordinate() {uTiling=2, vTiling=2}, Time=Multiply(A=@n50, B=@n51), Speed=Constant2Vector() {g=20})) {samplerType=SAMPLERTYPE_LinearGrayscale, texture=/Game/UI/HUD/Textures/UIPatterns/T_UI_GlitchNoise}
  n53 = Multiply(A=Subtract(A=ComponentMask(None=@n52.RGB) {g=true, r=true}) {constB=0.5}, B=Constant() {desc=Distortion Amount, r=0.02}) {constB=2}
  n54 = Time()
  n55 = Constant() {r=0.2}
  n56 = SmoothStep(Value=Frac(None=Multiply(A=@n54, B=@n55) {constB=0.5})) {constMin=0.9}
  NamedRerouteDeclaration(None=Add(A=ComponentMask(None=@n49) {g=true, r=true}, B=Multiply(A=@n53, B=@n56))) {name=Icon UVs, nodeColor=<rgba 0 0 0 1>}
  NamedRerouteDeclaration(None=@n19.Result) {nodeColor=<rgba 0.4062 0 1 1>}
  NamedRerouteDeclaration(None=@n13) {name=Blue Channel, nodeColor=<rgba 0 0 0 1>}
  n57 = LinearInterpolate(A=TextureCoordinate(), B=NamedRerouteUsage(), Alpha=ScalarParameter() {parameterName=Glitch?})
  n58 = Constant() {r=0.4}
  n59 = MaterialFunctionCall(
   UVs=Add(
    A=MaterialFunctionCall(UVs=@n57, Texture Scale=@n58) {materialFunction=&ScaleUVsByCenter}.UVs
    B=Constant2Vector()
   )
   "Rotation Angle (0-1)"=Constant()
  ) {materialFunction=&CustomRotator}
  n60 = Constant() {r=6}
  n61 = MaterialFunctionCall(UVs=@n59.Rotated Values, Radius=Constant() {r=0.9}, Sides=@n60) {materialFunction=&MF_SDF_RegularPolygon}
  NamedRerouteDeclaration(None=MaterialFunctionCall(DistanceField=@n61.SDF) {materialFunction=&MF_UI_AntiAliasing}.Inner) {name=Polygon, nodeColor=<rgba 0 0 0 1>}
  n62 = Constant() {r=0.035}
  n63 = Constant() {r=-3}
  n64 = Add(A=Add(A=NamedRerouteUsage(), B=NamedRerouteUsage()), B=NamedRerouteUsage())
  NamedRerouteDeclaration(None=Multiply(
    A=MaterialFunctionCall!(DistanceField=@n61.SDF, Threshold, Thickness=@n62, Alignment=@n63, Smoothing) {materialFunction=/FortUILibrary/MaterialFunctions/Shapes/Modifiers/MF_UI_ShapeOutline}.Outline
    B=OneMinus(None=@n64)
   )) {name=Outline, nodeColor=<rgba 0 0 0 1>}
  NamedRerouteDeclaration(None=MaterialFunctionCall() {materialFunction=&GetUserInterfaceUV}.Normalized UV) {name=Normalized UV, nodeColor=<rgba 0 0 0 1>}
  n65 = Constant() {r=0.82}
  NamedRerouteDeclaration(None=MaterialFunctionCall(DistanceField=MaterialFunctionCall(UVs=@n59.Rotated Values, Radius=@n65, Sides=@n60) {materialFunction=&MF_SDF_RegularPolygon}.SDF) {materialFunction=&MF_UI_AntiAliasing}.Inner) {name=Polygon Small, nodeColor=<rgba 0 0 0 1>}
  n66 = NamedRerouteUsage()
  NamedRerouteDeclaration(None=Multiply(A=NamedRerouteUsage(), B=Power(Base=@n66) {constExponent=3})) {name=Polygon Gradient, nodeColor=<rgba 0 0 0 1>}
  NamedRerouteUsage()
  ScalarParameter()
 ]
}
