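# UI material for the ToyFox boss in TidalCrane that draws a soft radial glow with time-cycling party colors.
# Time and constants feed a cosine gradient function, whose offset result is dimmed and masked by a powered exponential radial gradient.
# The same mask is also multiplied by a zero constant, which, with AlphaComposite blending, points to an additive-style glow.
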Material /TidalCrane_ToyFox_Boss/UI/M_UI_ToyFox_Glow
graph {
 properties {blendMode=BLEND_AlphaComposite, materialDomain=MD_UI, pixelDepthOffsetMode=PDOM_Legacy, usedShadingModels=Default Lit}
 nodes [
  n1 = Power(Base=MaterialFunctionCall() {materialFunction=&RadialGradientExponential}.RadialGradientExponential) {constExponent=3}
  Multiply(A=@n1, B=Constant())
  n2 = Multiply(A=Time())
  n3 = Constant() {r=0.3}
  n4 = Constant()
  n5 = Add(A=MaterialFunctionCall!(
    Gradient
    a
    b
    c
    d=MaterialFunctionCall(R=@n2, G=@n3, B=@n4) {materialFunction=&AppendMany}.RGB
   ) {materialFunction=/FortUILibrary/MaterialFunctions/Color/MF_UI_CosineGradient}.Result) {constB=0.5}
  Multiply(A=Multiply(A=NamedRerouteDeclaration(None=@n5) {name=Party Time Colorz, nodeColor=<rgba 0 0 0 1>}, B=Constant() {r=0.4}), B=@n1)
 ]
}
